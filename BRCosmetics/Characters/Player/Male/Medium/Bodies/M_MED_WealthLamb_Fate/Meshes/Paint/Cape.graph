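# Cape material in the Paint mesh folder of the male medium player body cosmetic M_MED_WealthLamb_Fate.
# It is a Default Lit material whose only node is a constant color of muted periwinkle blue, with no link to a material output shown.
# It has no textures or parameters and carries Interchange import data, which suggests a placeholder generated during mesh import.
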Material /BRCosmetics/Characters/Player/Male/Medium/Bodies/M_MED_WealthLamb_Fate/Meshes/Paint/Cape
graph {
 properties {assetImportData="<self>:InterchangeAssetImportData_0", usedShadingModels=Default Lit}
 nodes [Constant3Vector() {constant=<rgba 0.2147 0.2009 0.5595 1>}]
}
